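; Gameplay ability for a team spray weapon that grants shield and low-gravity effects, filed under the LightWeight augment.
; A montage trigger starts a looping timer that, on the server, commits the ability, applies both effects to the user and an effect container to selected targets.
; Input release or empty ammo, after an EndSpray outro, ends the ability.
; Ending clears FX, spray cue and timer, and removes the item if the outro played.
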(event Custom|ApplyEffectIncremental
  ;[0 0.5221 1] Look for targets and apply effect
  (if (Networking|IsServer)
    (bind _returnvalue (Ability|CommitAbility))
    (if _returnvalue
      (bind (_asynctaskproxy _targetdata) (Ability|Tasks|DoTargetSelection "None" "(TagName=\"Targeting.ApplicationTag.0\")" true)
        (:then
          (Fortnite|ApplyGameplayEffectsToActor (Ability|GetAvatarActorfromActorInfo) (Utilities|Array|MakeArray "/PaprikaConsumables/Gameplay/TeamSpray/GE_Paprika_TeamSpray_Shields.GE_Paprika_TeamSpray_Shields_C" "/PaprikaConsumables/Gameplay/TeamSpray/GE_Paprika_TeamSpray_LowGrav.GE_Paprika_TeamSpray_LowGrav_C") (Transformation|GetActorLocation (Ability|GetAvatarActorfromActorInfo)) false 0 (Ability|GetActivatingAbilityComponent)))
        (:Targeted
          (Ability|ApplyGameplayEffectContainer _targetdata "(TagName=\"Targeting.ApplicationTag.0\")")))
      (else
        (Ability|EndAbility)))))

(event Ability|EventActivateAbility
  ;[0 0.5221 1] Activate ability and wait for spray section
  (Variables|Default|SetFizzWeap (Utilities|Casting|CastToB_Paprika_TeamSpray_Weap (Ability|GetCurrentSourceWeapon)))
  (Ability|Tasks|PlayMontageandWaitforTrigger "None" (Variables|Default|GetSprayMontage)
    (:then
      (EventDispatchers|BindEventtoOnLocalAmmoChanged (Ability|GetCurrentSourceWeapon) (EventDispatchers|CreateEvent :!event Custom|OnWeaponAmmoChanged))
      (Class|BPaprikaTeamSprayWeap|ToggleSprayFX (Variables|Default|GetFizzWeap) true))
    (:Triggered
      (Variables|Default|SetStartedSpraying? true)
      (bind _returnvalue (Utilities|Time|SetTimerbyEvent (EventDispatchers|CreateEvent :!event Custom|ApplyEffectIncremental) (ScalableFloat|GetValueatLevel (Variables|Default|GetHF_TickRate)) true))
      (Variables|Default|SetApplicationTimer _returnvalue)
      (CallFunction|ApplyEffectIncremental)
      (bind _returnvalue_2 (Ability|ApplyGameplayEffectToOwner "/PaprikaConsumables/Gameplay/TeamSpray/GE_Paprika_TeamSpray_SprayCue.GE_Paprika_TeamSpray_SprayCue_C"))
      (Variables|Default|SetTeamSpraySpec _returnvalue_2)
      (Ability|Tasks|WaitInputRelease true
        (:OnRelease
          (Ability|EndAbility))))
    (:Cancelled
      _)
    (:Completed
      _))
  (Ability|EndAbility))

(event Custom|ServerEndHealing
  ;[0 0.5221 1] End GE and timer on server
  (Ability|RemoveGameplayEffectFromOwnerWithHandle (Variables|Default|GetTeamSpraySpec))
  (Utilities|Time|ClearandInvalidateTimerbyHandle (Variables|Default|GetApplicationTimer)))

(event Ability|EventOnEndAbility
  (params (bWasCancelled :type Boolean))
  ;[1 0.1038 0.05208] End Ability
  (Class|BPaprikaTeamSprayWeap|ToggleSprayFX (Variables|Default|GetFizzWeap) false)
  (if (Variables|Default|GetStartedSpraying?)
    (Ability|RemoveGameplayEffectFromOwnerWithHandle (Variables|Default|GetTeamSpraySpec))
    (Utilities|Time|ClearandInvalidateTimerbyHandle (Variables|Default|GetApplicationTimer))
    (if (Variables|Default|GetPlayingOutro?)
      (CallFunction|RemoveTeamSpray))))

(event Custom|OnWeaponAmmoChanged
  (params (LocalCount :type Integer) (LocalRemaining :type Integer))
  ;[0.03429 0.3069 0.7899] Handle ammo change + input release
  (if (<= LocalRemaining 0)
    (Ability|Tasks|SetNextMontageSectionandWait (Animation|GetCurrentMontage) "EndSpray"
      (:then
        (Variables|Default|SetPlayingOutro? true))
      (:OnComplete
        (Ability|EndAbility)))))

(event Custom|HandleRelease
  ;[0.03429 0.3069 0.7899] Handle ammo change + input release
  (Ability|HasAuthority
    (:False
      (if (not (Variables|Default|GetStartedSpraying?))
        (Utilities|FlowControl|WaitforBothExecutions true
          (:Then
            _)))
      (if (not (Variables|Default|GetPlayingOutro?))
        (Ability|EndAbility)))))

(event Custom|EnableRelease
  ;[0.03429 0.3069 0.7899] Handle ammo change + input release
  (Utilities|FlowControl|WaitforBothExecutions true
    (:Then
      (if (not (Variables|Default|GetPlayingOutro?))
        (Ability|EndAbility)))))

(fn ClientIsEmpty?
  (return (<= (FortWeapon|GetRemainingAmmo (Ability|GetCurrentSourceWeapon)) 5)))

(fn GetCustomAbilitySourceTransform
  (bind (_returnvalue_x _returnvalue_y _returnvalue_z) (Math|Vector|BreakVector (Transformation|GetSocketLocation (Variables|/Script/Engine.Character|GetMesh (CallFunction|GetAvatarasFortPlayerPawn)) "root")))
  (return (Math|Vector|MakeVector (.x (Transformation|GetActorLocation (CallFunction|GetAvatarasFortPlayerPawn))) (.y (Transformation|GetActorLocation (CallFunction|GetAvatarasFortPlayerPawn))) _returnvalue_z) "0, 0, 0" "1.000000,1.000000,1.000000"))

(fn RemoveTeamSpray
  (Inventory|RemoveFortItemFromPlayer (Fortnite|GetFortPlayerControllerfromActor (Ability|GetActivatingPawn)) (FortWeapon|GetInventoryItem (Ability|GetCurrentSourceWeapon)) 1 true))
